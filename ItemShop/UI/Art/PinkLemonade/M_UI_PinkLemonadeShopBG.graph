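# UI-domain material for the Pink Lemonade Item Shop background, showing a swirling pink-lemonade liquid.
# A panning swirl flow texture distorts the scaled UVs of the liquid texture, which blends over a color-atlas curve, then gets hue-shifted and saturated.
# Separate branches, unconnected to that chain, hold an angled gradient blend and a Headband pattern with panning noise, screen-aligned UVs and optional perspective.
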Material /ItemShop/UI/Art/PinkLemonade/M_UI_PinkLemonadeShopBG
graph {
 properties {bUsedWithStaticMesh=false, materialDomain=MD_UI, pixelDepthOffsetMode=PDOM_Legacy, usedShadingModels=Default Lit}
 nodes [
  n1 = LinearInterpolate(A=Constant3Vector() {constant=<rgba 1 0.3364 0.8623 0>}, B=Constant3Vector() {constant=<rgba 0.7292 0 0.7195 0>}) {constAlpha=0.4002}
  n2 = MaterialFunctionCall(Texture Scale=ScalarParameter() {defaultValue=5.3031, group=Gradient, parameterName=GradientScale}) {materialFunction=&ScaleUVsByCenter}
  n3 = ScalarParameter() {defaultValue=8.6841, group=Gradient, parameterName=GradientAngle}
  n4 = CurveAtlasRowParameter(CurveTime=ComponentMask(None=MaterialFunctionCall(UVs=@n2.UVs, "Rotation Angle (0-1)"=@n3) {materialFunction=&CustomRotator}.Rotated Values) {r=true}) {
   atlas=/Game/UI/Frontend/ItemShop/Art/ItemShopColorAtlas
   curve=/Game/UI/Frontend/ItemShop/Art/PinkLemonade/ItemShop_PinkLemonadeBG_Curve
   defaultValue=9
   parameterName=CurveParamsBG
   sortPriority=0
  }
  n5 = LinearInterpolate(A=@n1, Alpha=Multiply()) {constA=-1.2391, constB=2.2813}
  LinearInterpolate(
   A=TextureSample() {texture=&/T_UI_PinkLemonade_Liquid}.RGB
   B=MaterialFunctionCall(
    Hue Shift Percentage=ScalarParameter() {defaultValue=0.0399, parameterName=HueShiftValue}
    Texture=LinearInterpolate(A=@n1, B=@n4, Alpha=@n5) {constAlpha=0.3343}
   ) {materialFunction=&HueShift}.Result
  ) {constAlpha=1.0673}
  n6 = MaterialFunctionCall(Float2=MaterialFunctionCall() {materialFunction=&ScreenResolution}.Visible Resolution) {materialFunction=&BreakOutFloat2Components}
  n7 = Max(A=@n6.G)
  Multiply(
   A=AppendVector(A=Divide(A=@n6.R, B=@n7), B=Constant() {r=1})
   B=MaterialFunctionCall() {materialFunction=&ScreenAlignedUVs}."X 100%, Y 100%"
  )
  n8 = ScalarParameter() {defaultValue=1, group=Headband, parameterName=Min, sliderMax=1}
  n9 = ScalarParameter() {defaultValue=0.104, group=Headband, parameterName=Max, sliderMax=1}
  n10 = VectorParameter() {group=Headband, parameterName=Streaks, channelNames {a="", b=Speed, g=StretchY, r=StretchX}, defaultValue=<rgba 0.7329 0.0052 0.005 1>}
  n11 = MaterialFunctionCall(UVs=NamedRerouteUsage(), "Rotation Angle (0-1)"=ScalarParameter() {defaultValue=-2.6248, group=Headband, parameterName=Rotation Angle}) {materialFunction=&CustomRotator}
  n12 = Panner(
   Coordinate=Multiply(A=@n11.Rotated Values, B=AppendVector(A=@n10.StretchX, B=@n10.StretchX))
   Speed=AppendVector(A=Constant() {r=-3.4954}, B=@n10.Speed)
  )
  n13 = VectorParameter() {group=Headband, parameterName=Noise, channelNames {a=Power, b=Speed, g=StretchY, r=StretchX}, defaultValue=<rgba 3.6698 0.0826 0.01 -0.3543>}
  n14 = AppendVector(A=@n13.StretchX, B=@n13.StretchX)
  n15 = Constant() {r=0.075}
  n16 = TextureSampleParameter2D(UVs=Panner(Coordinate=Multiply(A=@n11.Rotated Values, B=@n14), Speed=AppendVector(A=@n15, B=@n13.Speed))) {
   group=Headband
   parameterName=NoisePattern
   samplerType=SAMPLERTYPE_LinearColor
   texture=&HighContrastNoise
  }
  n17 = Power(Base=@n16.R, Exp=@n13.Power)
  n18 = VectorParameter() {group=Headband, parameterName=Noise2, channelNames {a=Power, b=Speed, g=StretchY, r=StretchX}, defaultValue=<rgba 0.6505 0.0398 0.1 0.3605>}
  n19 = AppendVector(A=@n18.StretchX, B=@n18.StretchX)
  n20 = Constant()
  n21 = TextureSampleParameter2D(UVs=Panner(Coordinate=Multiply(A=@n11.Rotated Values, B=@n19), Speed=AppendVector(A=@n20, B=@n18.Speed))) {
   group=Headband
   parameterName=NoisePattern
   samplerType=SAMPLERTYPE_LinearColor
   texture=&HighContrastNoise
  }
  n22 = Power(Base=@n21.R, Exp=@n18.Power)
  n23 = Constant()
  n24 = ScalarParameter() {defaultValue=0.1714, group=Headband, parameterName=GradientOffset}
  n25 = ComponentMask(None=Add(A=TextureCoordinate(), B=AppendVector(A=@n23, B=@n24))) {g=true}
  n26 = ScalarParameter() {defaultValue=0.7813, group=Headband, parameterName=DarkenGradient}
  n27 = Multiply(
   A=Add(
    A=TextureSampleParameter2D(UVs=@n12) {group=Headband, parameterName=Pattern, texture=/Game/Global/Textures/Default/PointSplatter/T_UI_StarBurst_Scatter}.R
    B=Multiply(A=@n17, B=@n22)
   )
   B=OneMinus(None=Multiply(A=@n25, B=@n26))
  )
  n28 = CurveAtlasRowParameter(CurveTime=ComponentMask(None=SmoothStep(Min=@n8, Max=@n9, Value=@n27)) {r=true}) {
   atlas=&ItemShopColorAtlas
   curve=&ItemShop_PinkLemonadeBG_Curve
   defaultValue=9
   parameterName=CurveParamsBG
   sortPriority=0
  }
  NamedRerouteDeclaration(None=LinearInterpolate(A=@n28, B=TextureSample() {texture=/ItemShop/UI/Art/Alfredo/T_UI_Alfredo_Mask}.RGB) {constAlpha=-0.316}) {name=Gradient_PY, nodeColor=<rgba 0 0 0 1>}
  n29 = VectorParameter() {group=Headband, parameterName=HeadbandScale, channelNames {a=OffsetY, b=OffsetX, g=Y, r=X}, defaultValue=<rgba 1 1 0 0>}
  n30 = MaterialFunctionCall!(
   Texture Scale=AppendVector(A=@n29.X, B=@n29.X)
   UVs=NamedRerouteUsage()
   Offset XY Texturesize Z=MaterialFunctionCall(R=@n29.OffsetX, G=@n29.OffsetX, B=Constant() {r=1}) {materialFunction=&AppendMany}.RGB
  ) {materialFunction=/Engine/Functions/Engine_MaterialFunctions01/Texturing/ScaleUVsAroundPoint}
  n31 = VectorParameter() {group=Headband, parameterName=HeadbandPerspective, channelNames {a=YSlide, b=XSlide, g=Y, r=X}}
  n32 = MaterialFunctionCall!(UV=@n30.UVs, Texture, FixedScaleFactor_X, FixedScaleFactor_Y, Persp_Y=@n31.Y, Persp_X=@n31.Y, Slide_X=@n31.Y, Slide_Y=@n31.Y, Flip_Y, Flip_X) {materialFunction=/Game/UI/Foundation/Materials/MF_UI_BasicPerspective}
  NamedRerouteDeclaration(None=Add(A=StaticSwitchParameter(True=@n32.FinalUvs, False=@n30.UVs) {group=Headband, parameterName=HeadbandUsePerpsective}) {constB=0}) {name=NormalizedSSUVs, nodeColor=<rgba 1 0 0.7969 1>}
  n33 = MaterialFunctionCall(Float2=MaterialFunctionCall() {materialFunction=&GetUserInterfaceUV}.Pixel Size) {materialFunction=&BreakOutFloat2Components}
  n34 = Divide(A=@n33.G, B=@n33.G)
  n35 = Constant() {r=1}
  NamedRerouteDeclaration(None=MaterialFunctionCall(UVs=TextureCoordinate(), Texture Scale=AppendVector(A=@n34, B=@n35)) {materialFunction=&ScaleUVsByCenter}.UVs) {name=NormalizedUVs, nodeColor=<rgba 0.8047 1 0 1>}
  Reroute()
  n36 = CurveAtlasRowParameter() {
   atlas=&ItemShopColorAtlas
   curve=&ItemShop_PinkLemonadeBG_Curve
   defaultValue=9
   parameterName=CurveParamsBG
   sortPriority=0
  }
  n37 = Constant() {r=0.5}
  n38 = TextureCoordinate()
  n39 = Constant() {r=0.06}
  n40 = ComponentMask(None=TextureSample(
    UVs=Panner(Coordinate=@n38, Speed=@n39) {speedX=0.07, speedY=0.132}
    Tex=TextureObject() {samplerType=SAMPLERTYPE_LinearColor, texture=/Game/Global/Textures/Default/Flowmaps/T_SwirlFlow_02}
   ).R) {g=true, r=true}
  n41 = Multiply(A=Multiply(A=Constant2Vector() {g=1, r=-1}, B=Multiply(A=@n37, B=@n40)), B=ScalarParameter() {defaultValue=1.0497, parameterName=size})
  n42 = TextureCoordinate()
  n43 = TextureSample(UVs=MaterialFunctionCall(UVs=Add(A=@n41, B=@n42), Texture Scale=ScalarParameter() {defaultValue=2.0164, parameterName=Layer1Scale}) {materialFunction=&ScaleUVsByCenter}.UVs) {texture=&T_UI_PinkLemonade_Liquid}
  Saturate(None=MaterialFunctionCall(
    Hue Shift Percentage=ScalarParameter() {defaultValue=0.0399, parameterName=HueShiftValue}
    Texture=LinearInterpolate(A=@n36, B=@n43.RGB) {constAlpha=0.8481}
   ) {materialFunction=&HueShift}.Result)
  TextureCoordinate()
 ]
}
