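# UI material for Pink Lemonade item shop tiles, layering an animated shine pattern over gradient backgrounds.
# Two exponential radial gradients blend blue, pink and peach, then a frame mask, chosen by the TileSize static switch for wide or normal tiles, blends in the shine.
# A panning swirl flowmap distorts the center-scaled shine UVs.
# A separate branch blends two dark navy tones by texture mask and reuses that shine through a bar mask.
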Material /ItemShop/UI/Art/PinkLemonade/M_UI_PinkLemonade_ShopTiles
graph {
 properties {bUsedWithStaticMesh=false, materialDomain=MD_UI, pixelDepthOffsetMode=PDOM_Legacy, usedShadingModels=Default Lit}
 nodes [
  n1 = Constant3Vector() {constant=<rgba 0.2344 0.3875 1 1>}
  n2 = Constant3Vector() {constant=<rgba 0.9387 0.2298 0.6756 1>}
  n3 = MaterialFunctionCall(
   CenterPosition=Constant2Vector() {g=1.7116, r=0.5}
   Radius=ScalarParameter() {defaultValue=1.7178, parameterName=FirstLayerRadius}
   Density=ScalarParameter() {defaultValue=2.8679, parameterName=FirstLayerDensity}
  ) {materialFunction=&RadialGradientExponential}
  n4 = Constant2Vector() {g=1.509, r=0.5}
  n5 = ScalarParameter() {defaultValue=1.0572, parameterName=SecondGradientRadius}
  n6 = ScalarParameter() {defaultValue=3.3322, parameterName=2ndRadius}
  n7 = Constant2Vector() {g=1, r=-1}
  n8 = Panner(Coordinate=TextureCoordinate()) {speedX=0.07, speedY=0.132}
  n9 = TextureObject() {samplerType=SAMPLERTYPE_LinearColor, texture=/Game/Global/Textures/Default/Flowmaps/T_SwirlFlow_02}
  n10 = Multiply(A=Constant() {r=0.5}, B=ComponentMask(None=TextureSample(UVs=@n8, Tex=@n9).R) {g=true, r=true})
  n11 = Add(A=Multiply(A=Multiply(A=@n7, B=@n10)), B=TextureCoordinate())
  n12 = ScalarParameter() {defaultValue=1.9466, parameterName=Layer1Scale}
  n13 = Saturate(None=TextureSample(UVs=MaterialFunctionCall(UVs=@n11, Texture Scale=@n12) {materialFunction=&ScaleUVsByCenter}.UVs) {texture=/OfferCatalog/Art/PinkLemonade/T_UI_ShinePattern}.RGB)
  n14 = MaterialFunctionCall() {materialFunction=&MF_UI_UVScaling}
  LinearInterpolate(
   A=LinearInterpolate(
    A=LinearInterpolate(A=@n1, B=@n2, Alpha=@n3.RadialGradientExponential)
    B=Constant3Vector() {constant=<rgba 1 0.8228 0.6514 1>}
    Alpha=MaterialFunctionCall(CenterPosition=@n4, Radius=@n5, Density=@n6) {materialFunction=&RadialGradientExponential}.RadialGradientExponential
   )
   B=@n13
   Alpha=StaticSwitchParameter(
    True=TextureSampleParameter2D(UVs=@n14.ScaleX) {parameterName=WideTile, texture=/OfferCatalog/Art/PinkLemonade/T_UI_FrameWide}.R
    False=TextureSampleParameter2D(UVs=@n14.ScaleY) {parameterName=NormalTile, texture=/OfferCatalog/Art/PinkLemonade/T_UI_Frame}.R
   ) {parameterName=TileSize}
  )
  LinearInterpolate(
   A=LinearInterpolate(
    A=Constant3Vector() {constant=<rgba 0.1384 0.1812 0.3663 1>}
    B=Constant3Vector() {constant=<rgba 0.0185 0.0252 0.0704 1>}
    Alpha=TextureSample(UVs=@n14.ScaleX) {texture=/OfferCatalog/Art/PinkLemonade/T_UI_DualColor}.G
   )
   B=@n13
   Alpha=TextureSample(UVs=@n14.ScaleX) {texture=/OfferCatalog/Art/PinkLemonade/T_UI_ShineBars}.R
  )
 ]
}
